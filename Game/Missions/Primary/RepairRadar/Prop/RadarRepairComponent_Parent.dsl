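; RadarRepairComponent_Parent is a Blueprint parent class for repairable radar component props in the RepairRadar primary mission.
; It declares an OnRep_Repaired function, pointing to a replicated repaired state, and a construction script.
; Both functions are empty here.
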(fn OnRep_Repaired)

(fn ConstructionScript)
