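; Interactable solar panel prop for the Repair Radar mission that players use to register a repair.
; Interacting swaps its static mesh to the industrial solar panel, sends the native GoalItemReturned event to the Mission Manager with its mission GUID, then disables further interaction.
; CanInteract? gates interaction and has an empty OnRep handler, while the prompt text comes from the InteractText variable.
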(event Interaction|EventBlueprintOnInteract
  (params (InteractingPawn :type Fort%20Pawn%20Object%20Reference) (InteractionBeingAttempted :type EInteractionBeingAttempted%20Enum :default "FirstInteraction"))
  (Variables|Mesh|SetStaticMesh "/Game/Environments/Props/Industrial/Meshes/IND_PowerSolarPanel_01a.IND_PowerSolarPanel_01a")
  (CallFunction|MissionManager|ExecuteMissionEvent(v2) :EventType "(CategoryName=\"Native\",EventName=\"GoalItemReturned\")" :MissionGuid (Missions|GetMissionGuid :!class /Script/FortniteGame.FortSpawnableByPlacementSystemInterface) :DO_NOT_USE_THIS_OR_VARIABLES_BELOW false :EventInstigator self :GenericInt 0 :GenericFloat 0) ; GoalItemReturned
  (Variables|Default|SetCanInteract? false))

(fn BlueprintCanInteract
  (params (InteractingPawn :type Fort%20Pawn%20Object%20Reference) (InteractionBeingAttempted :type EInteractionBeingAttempted%20Enum :default "FirstInteraction") (InteractionType :type TInteractionType%20Enum :default "IT_NoInteraction"))
  (return (Variables|Default|GetCanInteract?)))

(fn BlueprintGetInteractionString
  (params (InteractingPawn :type Fort%20Pawn%20Object%20Reference) (InteractionBeingAttempted :type EInteractionBeingAttempted%20Enum :default "FirstInteraction"))
  (return (Variables|Default|GetInteractText)))

(fn OnRep_CanInteract?)

(fn ConstructionScript
  (|CallParentFunction))
